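# Unlit, two-sided alpha-composite material for the HangSpec skydiving contrail's Niagara ribbons and meshes, drawing a halftone-dotted trail.
# Screen-space dots anchor to object position and scale with view depth, with optional texture-cutout and Ben-Day sizing.
# They mask panning trail textures eroded by noise and depth-faded; NoDots blends to a side-tinted trail with Fresnel opacity.
# Dynamic parameters drive timing and vertex offsets combining clamped pull, sine wave, noise and SplineThicken.
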Material /BRCosmetics/Athena/Effects/Particles/Skydiving/NS_Contrail_HangSpec/Materials/M_Contrail_HangSpec_Dots
graph {
 properties {
  bAutomaticallySetUsageInEditor=false
  bUsedWithNiagaraMeshParticles=true
  bUsedWithNiagaraRibbons=true
  bUsedWithStaticMesh=false
  blendMode=BLEND_AlphaComposite
  pixelDepthOffsetMode=PDOM_Legacy
  shadingModel=MSM_Unlit
  twoSided=true
  usedShadingModels=Unlit
 }
 nodes [
  n1 = ConstantBiasScale(None=ScreenPosition().ViewportUV) {bias=-0.5, scale=2}
  n2 = MaterialFunctionCall(Float2=ViewSize()) {materialFunction=&BreakOutFloat2Components}
  n3 = AppendVector(A=Constant() {r=1}, B=Divide(A=@n2.G, B=@n2.G))
  n4 = Multiply(A=MaterialFunctionCall(World Position=ObjectPositionWS()) {materialFunction=&TransformToClipSpace}.Clip Space XY)
  n5 = Subtract(A=ConstantBiasScale(None=Multiply(A=@n1, B=@n3)), B=Multiply(A=Multiply(A=@n3, B=@n4)) {constB=0.5})
  n6 = ComponentMask(None=TransformPosition(None=ObjectPositionWS()) {transformSourceType=TRANSFORMPOSSOURCE_World, transformType=TRANSFORMPOSSOURCE_View}) {b=true}
  n7 = Multiply(A=Divide(B=@n6))
  n8 = Multiply(A=ScalarParameter() {defaultValue=100, parameterName=Tiling}, B=Divide(B=Clamp(None=@n6) {maxDefault=10000, minDefault=10}))
  n9 = Frac(None=Multiply(
    A=MaterialFunctionCall(UVs=@n5, Texture Scale=@n7) {materialFunction=&ScaleUVsByCenter}.UVs
    B=Clamp(None=@n8) {maxDefault=5, minDefault=0.1}
   ))
  n10 = SmoothStep(Value=Length(None=Subtract(A=@n9) {constB=0.5}) {is=class MaterialExpressionLength}) {constMax=0.5}
  n11 = TextureCoordinate()
  n12 = LinearInterpolate(
   A=ScalarParameter() {defaultValue=32, group=TextureCutout, parameterName=TextureResolution}
   B=DynamicParameter() {defaultValue=<rgba 32 1 1 1>, paramNames [Texture Resolution, Param2, Param3, Param4]}.Texture Resolution
   Alpha=ScalarParameter() {group=TextureCutout, parameterName=UseDynamicParam?}
  )
  n13 = Divide(A=Floor(None=Multiply(A=@n11, B=@n12) {constB=10}), B=@n12) {constB=10}
  n14 = Step(X=ChannelMaskParameter(None=TextureSampleParameter2D(UVs=@n13) {group=TextureCutout, parameterName=TextureCutoutInput, texture=/Game/Effects/Fort_Effects/Textures/T_PillowMill_Shapes}.RGBA) {group=TextureCutout, parameterName=TextureChannelSelect}.R) {constY=0.8}
  n15 = LinearInterpolate(
   A=Step(Y=@n10, X=ScalarParameter() {defaultValue=0.5, parameterName=DotSize}) {constX=0.5}
   B=Step(Y=@n10, X=LinearInterpolate(Alpha=@n14) {constB=0.5}) {constX=0.5}
   Alpha=ScalarParameter() {group=TextureCutout, parameterName=UseTextureCutout?}
  ) {constA=0.5}
  n16 = ScalarParameter() {group=BenDayDots, parameterName=SphereMaskDotScaleRange_Min}
  n17 = ScalarParameter() {defaultValue=0.6, group=BenDayDots, parameterName=SphereMaskDotScaleRange_Max}
  n18 = ObjectPositionWS()
  n19 = ScalarParameter() {defaultValue=0.25, group=BenDayDots, parameterName=SphereMaskRadius}
  n20 = ScalarParameter() {defaultValue=7, group=BenDayDots, parameterName=SphereMaskResolution}
  n21 = Floor(None=Multiply(
    A=MaterialFunctionCall!(Position=@n18, UVRadialScale=@n19) {materialFunction=/Game/Weapons/WeaponSkins/MaterialFunction/ObjectStabilizedScreenSpaceUVs}.StabilizedUVs
    B=@n20
   ) {constB=20})
  n22 = Power(Base=MaterialFunctionCall!(UVs=Divide(A=@n21, B=@n20) {constB=20}, CenterPosition, Radius) {materialFunction=/Engine/Functions/Engine_MaterialFunctions01/Gradient/SphereGradient-2D}.Result 0-1)
  n23 = Step(
   Y=@n10
   X=Add(
    A=ScalarParameter() {defaultValue=0.4, group=BenDayDots, parameterName=DotScaleMaster}
    B=LinearInterpolate(A=@n16, B=@n17, Alpha=@n22) {constB=0.6}
   ) {constA=0.2}
  )
  n24 = Divide(A=MaterialFunctionCall() {materialFunction=/Game/Effects/Fort_Effects/MaterialFunctions/MF_U}.U) {constB=0.1}
  n25 = ScalarParameter() {defaultValue=222, parameterName=DepthFadeDistanceA}
  n26 = ScalarParameter() {defaultValue=22, parameterName=DepthFadeDistanceB}
  n27 = Saturate(None=Subtract(A=@n24) {constB=0.25})
  n28 = TextureCoordinate() {uTiling=0.7, vTiling=0.8}
  n29 = ComponentMask(None=@n28) {r=true}
  n30 = OneMinus(None=@n29)
  n31 = DynamicParameter() {parameterIndex=3, paramNames [WPO_Strength,Max,TimeOffset,TimeNiagara]}
  n32 = Add(A=@n31.TimeNiagara, B=DynamicParameter() {parameterIndex=3, paramNames [WPO_Strength,Max,TimeOffset,TimeNiagara]}.TimeOffset)
  n33 = ScalarParameter() {defaultValue=-0.5, parameterName=Panner Speed X 2}
  n34 = ScalarParameter() {parameterName=Panner Speed Y 2}
  n35 = TextureSample(UVs=Panner(Coordinate=@n28, Time=@n32, Speed=AppendVector(A=@n33, B=@n34)) {speedX=-0.5}) {texture=/BRCosmetics/Athena/Effects/Particles/Skydiving/NS_Contrail_Lancelot/T_Lancelot_Trail}
  n36 = Add(A=@n35.R, B=@n35.R)
  n37 = TextureCoordinate()
  n38 = ComponentMask(None=@n37) {r=true}
  n39 = OneMinus(None=Multiply(A=@n38, B=@n38))
  n40 = Constant() {r=0.5}
  n41 = Subtract(A=ComponentMask(None=TextureCoordinate()) {g=true}, B=@n40)
  n42 = OneMinus(None=Abs(None=Divide(A=@n41, B=@n40)))
  n43 = ScalarParameter() {defaultValue=-1, parameterName=Noise Skew, sliderMax=1}
  n44 = AppendVector(A=Add(A=ComponentMask(None=@n37) {g=true}, B=Multiply(A=@n38, B=@n43)), B=@n38)
  n45 = AppendVector(A=ScalarParameter() {defaultValue=0.3, parameterName=Noise Scale U}, B=ScalarParameter() {defaultValue=1.5, parameterName=Noise Scale V})
  n46 = ScalarParameter() {defaultValue=-0.04, parameterName=Panner Speed X}
  n47 = ScalarParameter() {defaultValue=-1.66, parameterName=Panner Speed Y}
  n48 = TextureSample(UVs=Panner(Coordinate=Multiply(A=@n44, B=@n45), Time=@n32, Speed=AppendVector(A=@n46, B=@n47)) {bFractionalPart=true}) {
   samplerType=SAMPLERTYPE_LinearGrayscale
   texture=/BRCosmetics/Athena/Effects/Particles/Skydiving/NS_Contrail_DarkAndLight/T_DarkAndLight_Noise
  }
  n49 = SquareRoot(None=@n38)
  n50 = ScalarParameter() {defaultValue=2, parameterName=Noise Modulation, sliderMax=9}
  n51 = OneMinus(None=MaterialFunctionCall!(Gradient=Multiply(A=@n39, B=@n42), Noise=OneMinus(None=@n48.R), Modulation Amount=Multiply(A=@n49, B=@n50)) {materialFunction=/Game/Effects/Fort_Effects/MaterialFunctions/MF_GradientNoiseModulate}.Result)
  n52 = ScalarParameter() {defaultValue=0.5, parameterName=Noise Erosion, sliderMax=1}
  n53 = Saturate(None=OneMinus(None=MaterialFunctionCall!(Gradient=@n51, Erosion Amount=@n52) {materialFunction=/BRCosmetics/Athena/Effects/Particles/Skydiving/NS_Contrail_DarkAndLight/Functions/MF_ErosionNormalized}.Result))
  n54 = Multiply(A=Multiply(A=Multiply(A=@n30, B=@n36), B=@n53)) {constB=0.5}
  n55 = TextureSample(UVs=Panner(Coordinate=@n28, Time=@n32) {speedX=-1.5}) {texture=/BRCosmetics/Athena/Effects/Particles/Skydiving/NS_Contrail_Watercolor/T_Watercolor_Trail_Mesh}
  n56 = Multiply(A=@n55.R, B=Multiply(A=@n53, B=@n55.R))
  n57 = Multiply(A=DepthFade(Opacity=Saturate(None=@n24), FadeDistance=LinearInterpolate(A=@n25, B=@n26, Alpha=@n27)), B=Power(Base=Add(A=@n54, B=@n56)))
  n58 = Multiply(A=LinearInterpolate(A=@n15, B=Multiply(A=@n23, B=@n15), Alpha=ScalarParameter() {group=BenDayDots, parameterName=UseBenDayMask?}), B=@n57)
  n59 = ParticleColor()
  n60 = Abs(None=DotProduct(A=VertexNormalWS(), B=CameraVectorWS()))
  LinearInterpolate(
   A=@n58
   B=Multiply(
    A=Multiply(A=@n57, B=@n59.A) {constB=0}
    B=MaterialFunctionCall(Exponent=@n60) {materialFunction=&MF_Fresnel_FX}.Fresnel Double Sided
   )
   Alpha=ScalarParameter() {parameterName=NoDots}
  )
  n61 = TwoSidedSign()
  n62 = Constant()
  n63 = Multiply(A=VectorParameter() {parameterName=Color, defaultValue=<rgba 1 0.1655 0.4017 1>}.RGB)
  n64 = VectorParameter() {parameterName=Color_M, defaultValue=<rgba 1 0.0524 0.215 1>}
  n65 = Multiply(A=@n64.RGB)
  n66 = Multiply(
   A=@n57
   B=LinearInterpolate(
    A=If(A=@n61, B=@n62, "A > B"=@n63, "A == B"=@n64.RGB, "A < B"=@n65)
    B=If(A=@n61, B=@n62, "A > B"=@n65, "A == B"=@n64.RGB, "A < B"=@n63)
    Alpha=DynamicParameter() {parameterIndex=1, paramNames [WPO_Wave_Speed, Invert_Side, WPO_Wave_Size, UV_Tiling]}.Invert_Side
   )
  )
  MaterialFunctionCall(Emissive=LinearInterpolate(A=@n58, B=Multiply(A=@n66, B=@n59.A) {constB=0}, Alpha=ScalarParameter() {parameterName=NoDots})) {materialFunction=&MF_FX_EmissiveAdjustments}
  ObjectPositionWS()
  OneMinus(None=@n29)
  Power(Base=@n29) {constExponent=5}
  Power(Base=OneMinus(None=@n29)) {constExponent=2.5}
  n67 = DynamicParameter() {parameterIndex=2, paramNames [X,Y,Z,WPOIntensity]}
  n68 = MaterialFunctionCall(X=@n67.X, Y=@n67.X, Z=@n67.X) {materialFunction=&MakeFloat3}
  n69 = ParticlePositionWS()
  n70 = DynamicParameter() {parameterIndex=3, paramNames [WPO_Strength,Max,TimeOffset,TimeNiagara]}
  n71 = OneMinus(None=ComponentMask(None=TextureCoordinate()) {r=true})
  n72 = ScalarParameter() {defaultValue=2, group=WPO, parameterName=InterpolatePower}
  n73 = LinearInterpolate(
   B=Clamp(None=Subtract(A=@n68.Result, B=@n69), Min=Multiply(A=@n70.Max) {constA=-1, constB=-1}, Max=@n70.Max) {maxDefault=50, minDefault=-50}
   Alpha=Multiply(A=@n67.WPOIntensity, B=Power(Base=@n71, Exp=@n72) {constExponent=3})
  )
  n74 = ScalarParameter() {defaultValue=1, group=WPO, parameterName=Interpolate}
  n75 = ComponentMask(None=TextureCoordinate()) {r=true}
  n76 = DynamicParameter() {paramNames [Spline_Width, UV_Speed, WPO_Wave_Size, UV_Tiling]}
  n77 = DynamicParameter() {parameterIndex=1, paramNames [WPO_Wave_Speed, Invert_Side, WPO_Wave_Size, UV_Tiling]}
  n78 = Sine(None=Add(A=Multiply(A=@n75, B=@n76.WPO_Wave_Size) {constB=1.5}, B=Multiply(A=@n31.TimeNiagara, B=@n77.WPO_Wave_Speed)))
  n79 = Constant3Vector() {constant=<rgba 1 1 1 0>}
  n80 = Multiply(
   A=Transform(None=Multiply(A=@n78, B=@n79)) {transformSourceType=TRANSFORMSOURCE_View}
   B=Multiply(
    A=ScalarParameter() {defaultValue=14, group=WPO, parameterName=WPO_Wave}
    B=DynamicParameter() {parameterIndex=3, paramNames [WPO_Strength,Max,TimeOffset,TimeNiagara]}.WPO_Strength
   )
  ) {constB=10}
  n81 = Add(A=TextureCoordinate() {uTiling=0.1}, B=@n31.TimeNiagara)
  n82 = Multiply(
   A=Multiply(A=TextureSample(UVs=@n81) {texture=/Game/Effects/Fort_Effects/Textures/Power_Source/T_Power_Source_Sphere_Texture}.RGB, B=VertexNormalWS())
   B=ScalarParameter() {defaultValue=10, group=WPO, parameterName=WPO_Noise}
  ) {constB=12}
  n83 = StaticComponentMaskParameter(None=Subtract(A=WorldPosition().XYZ, B=ObjectPositionWS())) {defaultB=true}
  n84 = DynamicParameter() {paramNames [Spline_Width, UV_Speed, WPO_Wave_Size, UV_Tiling]}
  n85 = Multiply(A=ScalarParameter() {defaultValue=1, group=SplineThicken, parameterName=Spline_Base}, B=@n84.Spline_Width)
  n86 = Multiply(A=ScalarParameter() {group=SplineThicken, parameterName=Spline_Tip}, B=@n84.Spline_Width)
  n87 = Add(
   A=Multiply(A=Add(A=@n80, B=@n82), B=Divide(A=@n83) {constB=200})
   B=LinearInterpolate(
    B=MaterialFunctionCall(WidthBase=@n85, WidthTip=@n86) {materialFunction=&SplineThicken}.WorldPositionOffset
    Alpha=ScalarParameter() {defaultValue=1, group=SplineThicken, parameterName=SplineThicken}
   )
  )
  n88 = LinearInterpolate(A=Add(A=Multiply(A=@n73, B=@n74), B=@n87), B=@n87, Alpha=ScalarParameter() {group=WPO, parameterName=WPO_Offset})
  Reroute(None=@n88)
  Time()
  Time()
  Time()
  Transform(None=@n88) {transformSourceType=TRANSFORMSOURCE_View}
 ]
}
