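# Additive, unlit, two-sided material for a muzzle-flash-style Niagara mesh particle in the HangSpec skydiving contrail effect.
# A flash texture blends with a panning tornado pattern through dynamic parameters, then takes particle color.
# The Is Translucent switch toggles between that result and one with an added screen-space dot pattern, anchored to object position and scaled by depth.
# Opacity combines particle alpha with near-camera, view-angle and depth fades, which low quality skips.
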Material /BRCosmetics/Athena/Effects/Particles/Skydiving/NS_Contrail_HangSpec/Materials/M_HangSpec_Muzzle
graph {
 properties {
  bAutomaticallySetUsageInEditor=false
  bUsedWithNiagaraMeshParticles=true
  blendMode=BLEND_Additive
  opacityMaskClipValue=0.933
  pixelDepthOffsetMode=PDOM_Legacy
  shadingModel=MSM_Unlit
  twoSided=true
  usedShadingModels=Unlit
 }
 nodes [
  n1 = TextureSampleParameter2D() {parameterName=FlashTexture, texture=/Game/Weapons/FORT_Pistols/Effects/Muzzle/T_Muzzle_Flash_01_Sharp}
  n2 = DynamicParameter() {defaultValue=<rgba 0 1 1 1>, paramNames [TextureBlend,PanSpeed,UTiling,VTiling]}
  n3 = Panner(
   Coordinate=Multiply(A=TextureCoordinate() {uTiling=2, vTiling=0.5}, B=AppendVector(A=@n2.UTiling, B=@n2.UTiling))
   Time=Multiply(A=@n2.PanSpeed, B=Time())
  ) {speedY=1}
  n4 = LinearInterpolate(
   A=@n1.R
   B=Multiply(A=@n1.R, B=TextureSample(UVs=@n3) {texture=/Game/Global/Textures/Default/Tileable_Patterns/T_Tornado_Tile}.R)
   Alpha=@n2.TextureBlend
  )
  n5 = ParticleColor()
  n6 = Multiply(A=@n4, B=@n5.RGB)
  n7 = MaterialFunctionCall(Float2=ViewSize()) {materialFunction=&BreakOutFloat2Components}
  n8 = AppendVector(A=Constant() {r=1}, B=Divide(A=@n7.G, B=@n7.G))
  n9 = Multiply(A=ConstantBiasScale(None=ScreenPosition().ViewportUV) {bias=-0.5, scale=2}, B=@n8)
  n10 = ObjectPositionWS()
  n11 = Multiply(
   A=@n8
   B=Multiply(A=MaterialFunctionCall(World Position=@n10) {materialFunction=&TransformToClipSpace}.Clip Space XY)
  )
  n12 = ComponentMask(None=TransformPosition(None=ObjectPositionWS()) {transformSourceType=TRANSFORMPOSSOURCE_World, transformType=TRANSFORMPOSSOURCE_View}) {b=true}
  n13 = MaterialFunctionCall(UVs=Subtract(A=ConstantBiasScale(None=@n9), B=Multiply(A=@n11) {constB=0.5}), Texture Scale=Multiply(A=Divide(B=@n12))) {materialFunction=&ScaleUVsByCenter}
  n14 = Clamp(None=@n12) {maxDefault=10000, minDefault=10}
  n15 = Clamp(None=Multiply(A=ScalarParameter() {defaultValue=100, parameterName=Tiling}, B=Divide(B=@n14))) {maxDefault=5, minDefault=0.1}
  n16 = Subtract(A=Frac(None=Multiply(A=@n13.UVs, B=@n15))) {constB=0.5}
  n17 = Step(
   Y=SmoothStep(Value=Length(None=@n16) {is=class MaterialExpressionLength}) {constMax=0.5}
   X=ScalarParameter() {defaultValue=0.5, parameterName=DotSize}
  ) {constX=0.5}
  MaterialFunctionCall(Emissive=StaticSwitchParameter(True=@n6, False=Add(A=@n17, B=@n6)) {parameterName=Is Translucent}) {materialFunction=&MF_FX_EmissiveAdjustments}
  n18 = ScalarParameter() {defaultValue=150, parameterName=NearFadeDist}
  n19 = DotProduct(A=CameraVectorWS(), B=VertexNormalWS())
  n20 = Multiply(A=Power(Base=Abs(None=@n19))) {constB=4}
  n21 = DepthFade(
   Opacity=Multiply(
    A=MaterialFunctionCall(Distance=@n18) {materialFunction=&Near_Camera_Fade}.Result
    B=Clamp(None=@n20)
   )
   FadeDistance=ScalarParameter() {defaultValue=12, parameterName=DepthFadeDist}
  )
  n22 = QualitySwitch(Default=@n21, Low=Constant() {r=1}, High=@n21, Medium=@n21)
  StaticSwitchParameter(True=Multiply(A=@n4, B=Multiply(A=@n5.A, B=@n22)), False=Multiply(A=Multiply(A=@n1.R, B=@n5.A), B=@n22)) {parameterName=Is Translucent}
 ]
}
